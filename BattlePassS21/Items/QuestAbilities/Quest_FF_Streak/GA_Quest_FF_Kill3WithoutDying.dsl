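; Quest gameplay ability that tracks a streak of player kills without dying, for the FF Streak quest.
; On activation it waits on Killed and Died gameplay events, counting kills whose target casts to FortPlayerPawnAthena and zeroing the count on death.
; Each counted kill at or above the target number sends a Creative.Quest.FF.Streak custom stat event to the owner's quest manager.
; The count resets only on death, so further kills keep sending the event.
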(event Ability|EventOnEndAbility
  (params (bWasCancelled :type Boolean))
  (Ability|EndAbility))

(event Ability|EventFailedToActivatePassiveAbility
  (params (ActorInfo :type Gameplay%20Ability%20Actor%20Info%20Structure))
  (Ability|EndAbility))

(event Ability|EventActivateAbility
  (|CallParentFunction)
  (Ability|Tasks|WaitGameplayEvent "(TagName=\"Event.Damage.Died\")"
    (:then
      (bind (_asynctaskproxy _payload) (Ability|Tasks|WaitGameplayEvent "(TagName=\"Event.Damage.Killed\")")
        (:EventReceived
          (if (.bSuccess (Utilities|Casting|CastToFortPlayerPawnAthena (.Target (Utilities|Struct|BreakGameplayEventData _payload))))
            (Math|Integer|IncrementInt (Variables|Default|GetKills)
              (:%20%20
                (if (>= (Variables|Default|GetKills) (Variables|Default|GetTarget%20Kill%20Number))
                  (Quest|SendComplexCustomStatEvent (Variables|GAVars|Data|GetOwningPlayerQuestManager) 0 "(GameplayTags=((TagName=\"Creative.Quest.FF.Streak\")))" "(GameplayTags=)" 1))))))) ; When player kills, start counting how many they've killed
    )
    (:EventReceived
      (Variables|Default|SetKills 0))) ; Reset kills when player dies
)
